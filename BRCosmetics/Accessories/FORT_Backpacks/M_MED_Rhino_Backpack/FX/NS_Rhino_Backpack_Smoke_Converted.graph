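# Niagara smoke effect for the Rhino backpack cosmetic, puffing sub-UV sprites from an offset along negative Y.
# A movement emitter spawns by distance travelled and inherits owner velocity, while two standing emitters spawn only at near-zero speed by remapping inverse owner speed to spawn probability.
# A faint trail emitter thins as speed rises.
# Curves grow the sprites and mostly fade them, with the standing puffs also driving dynamic material parameters.
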NiagaraSystem /BRCosmetics/Accessories/FORT_Backpacks/M_MED_Rhino_Backpack/FX/NS_Rhino_Backpack_Smoke_Converted
dependencies {
 usedDataInterfaces [
  &NiagaraDataInterfaceCurve
  &NiagaraDataInterfaceVectorCurve
  &NiagaraDataInterfaceSpriteRendererInfo
 ]
 usedDynamicInputs []
 usedModules [&SystemState]
}
emitters {
 SmokeLoop_move001 {
  emitterUpdateScript [
   &EmitterState(Life Cycle Mode=::NewEnumerator1, Loop Duration=1, UseLoopDelay=true)
   &SpawnPerUnit(
    Module.Use Max Movement Threshold=true
    Module.Use Movement Tolerance=true
    Movement Tolerance=5
    Spawn Spacing:NiagaraFloat=&Divide_Float(A=1, B=&Divide_Float(A=7, B=50))
    Use Spawn Probability=false
   ) {module=SpawnPerUnit001}
   &SpawnPerUnit(
    Max Movement Threshold=1000
    Module.Use Max Movement Threshold=true
    Module.Use Movement Tolerance=true
    Movement Tolerance=0.1
    Spawn Probability:NiagaraFloat=&Multiply_Float(A=1, B=1)
    Spawn Spacing=1
    Use Spawn Probability=false
   ) {enabled=false}
  ]
  particleSpawnScript [
   /SetVariables(Particles.SpriteRotationRate=&RandomRangeFloat(Minimum=-0.1, Maximum=0.1, Evaluation Type=::NewEnumerator1)) {isSetParameters=true}
   &InitializeParticle(
    Color=<rgba 0.3573 0.3573 0.3573 0.2>
    Lifetime=0.3
    Sprite Rotation Angle=&RandomRangeFloat(Minimum=0, Maximum=1, Evaluation Type=::NewEnumerator1)
    Sprite Rotation Mode=::NewEnumerator1
    Sprite Size=&Vector2DFromFloat(Value=&RandomRangeFloat(Minimum=7, Maximum=12, Evaluation Type=::NewEnumerator1))
    Sprite Size Mode=::NewEnumerator4
    UsePositionOffset=false
    Write Material Random=true
   )
   &AddVelocity(
    Cone Angle=0
    Cone Angle Mode:ENiagara_AngleInput=::NewEnumerator1
    Cone Axis=<0 -1 0>
    Module.UseConeFalloff=false
    Module.UseVelDistribution=false
    Override Seed=false
    Velocity Mode=::NewEnumerator2
    Velocity Speed=30
   )
   &InheritVelocity(Inherited Velocity Amount Scale=<0.85 0.85 0>, Inherited Velocity Speed Limit=0, Speed Limit=false, UseSpeedThreshold=false)
   /CascadeToNiagaraConverter/NiagaraScripts/ModuleScripts/AddVectorToPosition(Vector:Vector3f=&TransformVector(Vector=<0 -12 -2>, Destination Space=::Simulation))
  ]
  particleUpdateScript [
   &ParticleState()
   &SpriteRotationRate(Rotation Rate=&AngleConversion(
     Angle:NiagaraFloat=$Particles.SpriteRotationRate
     Angle Input:ENiagara_AngleInput=::NewEnumerator1
     Angle Output:ENiagara_AngleInput=::NewEnumerator0
    ))
   &ScaleSpriteSize(
    Scale Sprite Size Mode=::NewEnumerator0
    Uniform Scale Factor=&FloatFromCurve(FloatCurve=DataInterface() {propertyValues {Curve {keys [
         {arriveTangent=2.3881, interpMode=RCIM_Cubic, leaveTangent=2.3881, tangentMode=RCTM_User, time=0, value=0.7}
         {arriveTangent=1.8945, interpMode=RCIM_Cubic, leaveTangent=1.8945, time=0.2461, value=1.5371}
         <at 0.9499 2.4995 RCIM_Cubic>
        ]}}})
   )
   &ScaleColor(
    Scale Alpha=&FloatFromCurve(FloatCurve=DataInterface() {propertyValues {Curve {keys [
         {arriveTangent=2.2613, interpMode=RCIM_Cubic, leaveTangent=2.2613, tangentMode=RCTM_User, time=0.0002, value=0.6007}
         {arriveTangent=3.6875, interpMode=RCIM_Cubic, leaveTangent=-5.8389, tangentMode=RCTM_Break, time=0.1009, value=0.7898}
         {arriveTangent=-0.8784, interpMode=RCIM_Cubic, leaveTangent=-0.8784, time=0.4971, value=0.008}
         <at 1 0>
        ]}}})
    Scale RGB=&VectorFromCurve(VectorCurve=DataInterface() {propertyValues {XCurve {keys [<at 0 0.4583>,<at 1 1>]}, YCurve {keys [<at 0 0.4583>,<at 1 1>]}, ZCurve {keys [<at 0 0.4583>,<at 1 1>]}}})
    ScaleA=true
    ScaleRGB=true
   )
   &SubUVAnimation(
    End Frame Range Override=62
    "Sprite Renderer "=DataInterface() {propertyValues {SpriteRenderer=&self:SmokeLoop_move001_1.NiagaraSpriteRendererProperties_2}}
    Start Frame Range Override=63
    UseEndFrame=false
    UseStartFrame=false
   )
   &KillParticlesInVolume(
    Box Size:Vector3f=&Subtract_Vector(A=<60 60 60>, B=<-60 -60 -60>)
    Invert Volume=true
    Kill Shape=::NewEnumerator1
    Volume Origin:Vector3f=<0 0 0>
   ) {enabled=false}
   &SolveForcesAndVelocity(Clamp Velocity=false, Limit Acceleration=false)
  ]
  rendererClasses [&NiagaraSpriteRendererProperties]
  renderers [{rendererClass=&NiagaraSpriteRendererProperties}]
 }
 SmokeLoop_stand {
  emitterUpdateScript [
   &EmitterState(Life Cycle Mode=::NewEnumerator1, Loop Duration=0.5, UseLoopDelay=true)
   &SpawnRate(
    Spawn Probability=&RemapRange(
     Input Value To Remap=&Divide_Float(A=1, B=&VectorLength(Vector=$Engine.Owner.Velocity))
     Input Min=0.2
     Input Max=0.21
     Output Min=0
     Output Max=1
    )
    SpawnRate=5
    Use Spawn Probability=true
   ) {module=SpawnRate001}
  ]
  particleSpawnScript [
   &InitializeParticle(
    Sprite Rotation Angle=&RandomRangeFloat(Minimum=0.25, Maximum=0, Evaluation Type=::NewEnumerator1)
    Sprite Rotation Mode=::NewEnumerator1
    Sprite Size=&Vector2DFromFloat(Value=&RandomRangeFloat(Minimum=12, Maximum=15, Evaluation Type=::NewEnumerator1))
    Sprite Size Mode=::NewEnumerator4
    UsePositionOffset=false
    Write Material Random=true
   )
   &Color(Color=&MakeLinearColorFromVectorAndFloat("Vector (RGB)"=<0.227 0.2307 0.2232>, "Float (Alpha)"=0.8))
   &AddVelocity(
    Cone Angle=0
    Cone Angle Mode:ENiagara_AngleInput=::NewEnumerator1
    Cone Axis=<0 -1 0>
    Module.UseConeFalloff=false
    Module.UseVelDistribution=false
    Override Seed=false
    Velocity Mode=::NewEnumerator2
    Velocity Speed=80
   )
   &AddVectorToPosition(Vector:Vector3f=&TransformVector(Vector=<0 -15 -2>, Destination Space=::Simulation))
   &Color(Color=<rgba 1 0 0.6037 1>) {enabled=false, module=Color001}
  ]
  particleUpdateScript [
   &ParticleState()
   &ScaleSpriteSize(
    Scale Sprite Size Mode=::NewEnumerator0
    Uniform Scale Factor=&FloatFromCurve(FloatCurve=DataInterface() {propertyValues {Curve {keys [
         {arriveTangent=0.6107, interpMode=RCIM_Cubic, leaveTangent=0.6107, tangentMode=RCTM_User, time=-0.0002, value=0.9772}
         {arriveTangent=1.0849, interpMode=RCIM_Cubic, leaveTangent=1.0849, time=0.2799, value=1.2487}
         <at 0.9479 2.0058 RCIM_Cubic>
        ]}}})
   )
   &ScaleColor(
    Scale Alpha=&FloatFromCurve(FloatCurve=DataInterface() {propertyValues {Curve {keys [
         {arriveTangent=5.9578, interpMode=RCIM_Cubic, leaveTangent=5.9578, tangentMode=RCTM_User, time=0.0008, value=0.019}
         {arriveTangent=-3.5738, interpMode=RCIM_Cubic, leaveTangent=-3.5738, tangentMode=RCTM_User, time=0.1004, value=0.4914}
         {arriveTangent=-0.5463, interpMode=RCIM_Cubic, leaveTangent=-0.5463, time=0.2025, value=0.2242}
         <at 1 0>
        ]}}})
    Scale RGB=&VectorFromCurve(VectorCurve=DataInterface() {propertyValues {XCurve {keys [<at 0 0.4583>,<at 1 1>]}, YCurve {keys [<at 0 0.4583>,<at 1 1>]}, ZCurve {keys [<at 0 0.4583>,<at 1 1>]}}})
    ScaleA=true
    ScaleRGB=true
   )
   &DynamicMaterialParameters(
    Index 0 Param 1=&FloatFromCurve(FloatCurve=DataInterface() {propertyValues {Curve {keys [<at 0.3 1 RCIM_Cubic>,<at 1 0>]}}})
    Index 0 Param 2=&FloatFromCurve(FloatCurve=DataInterface() {propertyValues {Curve {keys [<at 0 1 RCIM_Cubic>,<at 0.8 8.7408 RCIM_Cubic>]}}})
    Index 0 Param 3=1
    Index 0 Param 4=&FloatFromCurve(FloatCurve=DataInterface() {propertyValues {Curve {keys [<at 0.3 0>,<at 0.5 0.5>]}}})
    Module.Param0WriteEnabled=true
    Module.Param1WriteEnabled=true
    Module.Param2WriteEnabled=true
    Module.Param3WriteEnabled=true
   )
   &Drag(Drag=5, Use Linear Drag=true, Use Rotational Drag=true)
   &SolveForcesAndVelocity(Clamp Velocity=false, Limit Acceleration=false)
   &SubUVAnimation(
    Frames Per Second=25
    Playback Mode:ENiagara_FPSPlayrate=::NewEnumerator1
    Random Start Frame=true
    "Sprite Renderer "=DataInterface() {propertyValues {SpriteRenderer=&self:SmokeLoop_stand_1.NiagaraSpriteRendererProperties_0}}
    SubUV Animation Mode:ENiagara_SubUVLookupModeV2=::NewEnumerator3
    UseEndFrame=false
    UsePlayRateScale=false
    UseStartFrame=false
   )
  ]
  rendererClasses [&NiagaraSpriteRendererProperties]
  renderers [{rendererClass=&NiagaraSpriteRendererProperties}]
 }
 SmokeLoop_stand002 {
  emitterUpdateScript [
   &EmitterState(Life Cycle Mode=::NewEnumerator1, Loop Duration=0.5, UseLoopDelay=true)
   &SpawnRate(
    Spawn Probability=&RemapRange(
     Input Value To Remap=&Divide_Float(A=1, B=&VectorLength(Vector=$Engine.Owner.Velocity))
     Input Min=0.2
     Input Max=0.21
     Output Min=0
     Output Max=1
    )
    SpawnRate=10
    Use Spawn Probability=true
   ) {module=SpawnRate001}
  ]
  particleSpawnScript [
   &InitializeParticle(
    Sprite Rotation Angle=&RandomRangeFloat(Minimum=0.25, Maximum=0, Evaluation Type=::NewEnumerator1)
    Sprite Rotation Mode=::NewEnumerator1
    Sprite Size=&Vector2DFromFloat(Value=&RandomRangeFloat(Minimum=12, Maximum=15, Evaluation Type=::NewEnumerator1))
    Sprite Size Mode=::NewEnumerator4
    UsePositionOffset=false
    Write Material Random=true
   )
   &Color(Color=&MakeLinearColorFromVectorAndFloat("Vector (RGB)"=<0.227 0.2307 0.2232>, "Float (Alpha)"=0.8))
   &AddVelocity(
    Cone Angle=0
    Cone Angle Mode:ENiagara_AngleInput=::NewEnumerator1
    Cone Axis=<0 -1 0>
    Module.UseConeFalloff=false
    Module.UseVelDistribution=false
    Override Seed=false
    Velocity Mode=::NewEnumerator2
    Velocity Speed=80
   )
   &AddVectorToPosition(Vector:Vector3f=&TransformVector(Vector=<0 -15 -2>, Destination Space=::Simulation))
   &Color(Color=<rgba 1 0 0.6037 1>) {enabled=false, module=Color001}
  ]
  particleUpdateScript [
   &ParticleState()
   &ScaleSpriteSize(
    Scale Sprite Size Mode=::NewEnumerator0
    Uniform Scale Factor=&FloatFromCurve(FloatCurve=DataInterface() {propertyValues {Curve {keys [
         {arriveTangent=0.6107, interpMode=RCIM_Cubic, leaveTangent=0.6107, tangentMode=RCTM_User, time=-0.0002, value=0.9772}
         {arriveTangent=1.0849, interpMode=RCIM_Cubic, leaveTangent=1.0849, time=0.2799, value=1.2487}
         <at 0.9479 2.0058 RCIM_Cubic>
        ]}}})
   )
   &ScaleColor(
    Scale Alpha=&FloatFromCurve(FloatCurve=DataInterface() {propertyValues {Curve {keys [
         {arriveTangent=5.9578, interpMode=RCIM_Cubic, leaveTangent=5.9578, tangentMode=RCTM_User, time=0.0008, value=0.019}
         {arriveTangent=-3.5738, interpMode=RCIM_Cubic, leaveTangent=-3.5738, tangentMode=RCTM_User, time=0.1004, value=0.4914}
         {arriveTangent=-0.5463, interpMode=RCIM_Cubic, leaveTangent=-0.5463, time=0.2025, value=0.2242}
         <at 1 0>
        ]}}})
    Scale RGB=&VectorFromCurve(VectorCurve=DataInterface() {propertyValues {XCurve {keys [<at 0 0.4583>,<at 1 1>]}, YCurve {keys [<at 0 0.4583>,<at 1 1>]}, ZCurve {keys [<at 0 0.4583>,<at 1 1>]}}})
    ScaleA=true
    ScaleRGB=true
   )
   &DynamicMaterialParameters(
    Index 0 Param 1=&FloatFromCurve(FloatCurve=DataInterface() {propertyValues {Curve {keys [<at 0.3 1 RCIM_Cubic>,<at 1 0>]}}})
    Index 0 Param 2=&FloatFromCurve(FloatCurve=DataInterface() {propertyValues {Curve {keys [<at 0 1 RCIM_Cubic>,<at 0.8 8.7408 RCIM_Cubic>]}}})
    Index 0 Param 3=1
    Index 0 Param 4=&FloatFromCurve(FloatCurve=DataInterface() {propertyValues {Curve {keys [<at 0.3 0>,<at 0.5 0.5>]}}})
    Module.Param0WriteEnabled=true
    Module.Param1WriteEnabled=true
    Module.Param2WriteEnabled=true
    Module.Param3WriteEnabled=true
   )
   &Drag(Drag=5, Use Linear Drag=true, Use Rotational Drag=true)
   &SolveForcesAndVelocity(Clamp Velocity=false, Limit Acceleration=false)
   &SubUVAnimation(
    Frames Per Second=25
    Playback Mode:ENiagara_FPSPlayrate=::NewEnumerator1
    Random Start Frame=true
    "Sprite Renderer "=DataInterface() {propertyValues {SpriteRenderer=&self:SmokeLoop_stand002_0.NiagaraSpriteRendererProperties_0}}
    SubUV Animation Mode:ENiagara_SubUVLookupModeV2=::NewEnumerator3
    UseEndFrame=false
    UsePlayRateScale=false
    UseStartFrame=false
   )
  ]
  rendererClasses [&NiagaraSpriteRendererProperties]
  renderers [{rendererClass=&NiagaraSpriteRendererProperties}]
 }
 SmokeLoop_trail_stand001 {
  emitterUpdateScript [
   &EmitterState(Life Cycle Mode=::NewEnumerator1, Loop Duration=1, UseLoopDelay=true)
   &SpawnRate(SpawnRate=15, Use Spawn Probability=false) {enabled=false}
   &SpawnRate(
    Spawn Probability=&RemapRange(
     Input Value To Remap=&Divide_Float(A=1, B=&VectorLength(Vector=$Engine.Owner.Velocity))
     Input Min=0
     Input Max=0.1
     Output Min=0
     Output Max=1
    )
    SpawnRate=10
    Use Spawn Probability=true
   ) {module=SpawnRate001}
   &SpawnPerUnit(
    Module.Use Max Movement Threshold=true
    Module.Use Movement Tolerance=true
    Movement Tolerance=0.1
    Spawn Spacing:NiagaraFloat=&Divide_Float(A=1, B=&Divide_Float(A=0, B=50))
    Use Spawn Probability=false
   ) {enabled=false}
  ]
  particleSpawnScript [
   /SetVariables(Particles.SpriteRotationRate=&RandomRangeFloat(Minimum=-0.1, Maximum=0.1, Evaluation Type=::NewEnumerator1)) {isSetParameters=true}
   &InitializeParticle(
    Color=<rgba 0.7448 0.7448 0.7448 1>
    Lifetime=0.5
    Sprite Rotation Angle=&RandomRangeFloat(Minimum=0, Maximum=1, Evaluation Type=::NewEnumerator1)
    Sprite Rotation Mode=::NewEnumerator1
    Sprite Size=&Vector2DFromFloat(Value=&RandomRangeFloat(Minimum=7, Maximum=8.5, Evaluation Type=::NewEnumerator1))
    Sprite Size Mode=::NewEnumerator4
    UsePositionOffset=false
    Write Material Random=true
   )
   &Color(Color=&MakeLinearColorFromVectorAndFloat("Vector (RGB)"=<0.1644 0.1667 0.1622>, "Float (Alpha)"=0.07))
   &Color(Color=<rgba 1 0 0.6037 1>) {enabled=false, module=Color001}
   &AddVelocity(
    Cone Angle=0
    Cone Angle Mode:ENiagara_AngleInput=::NewEnumerator1
    Cone Axis=<0 -1 0>
    Module.UseConeFalloff=false
    Module.UseVelDistribution=false
    Override Seed=false
    Velocity Mode=::NewEnumerator2
    Velocity Speed=40
   )
   &InheritVelocity(Inherited Velocity Amount Scale=<0.5 0.5 0.5>, Inherited Velocity Speed Limit=0, Speed Limit=false, UseSpeedThreshold=false)
   &AddVectorToPosition(Vector:Vector3f=&TransformVector(Vector=<0 -15 -2>, Destination Space=::Simulation))
  ]
  particleUpdateScript [
   &ParticleState()
   &SpriteRotationRate(Rotation Rate=&AngleConversion(
     Angle:NiagaraFloat=$Particles.SpriteRotationRate
     Angle Input:ENiagara_AngleInput=::NewEnumerator1
     Angle Output:ENiagara_AngleInput=::NewEnumerator0
    ))
   &ScaleSpriteSize(
    Scale Sprite Size Mode=::NewEnumerator0
    Uniform Scale Factor=&FloatFromCurve(FloatCurve=DataInterface() {propertyValues {Curve {keys [
         {arriveTangent=2.3881, interpMode=RCIM_Cubic, leaveTangent=2.3881, tangentMode=RCTM_User, time=0.0001, value=0.7509}
         {arriveTangent=1.8411, interpMode=RCIM_Cubic, leaveTangent=1.8411, time=0.2461, value=1.5371}
         <at 0.9499 2.4995 RCIM_Cubic>
        ]}}})
   )
   &ScaleColor(
    Scale Alpha=&FloatFromCurve(FloatCurve=DataInterface() {propertyValues {Curve {keys [
         {arriveTangent=2.2613, interpMode=RCIM_Cubic, leaveTangent=2.2613, tangentMode=RCTM_User, time=0.0002, value=0.772}
         {arriveTangent=3.6875, interpMode=RCIM_Cubic, leaveTangent=-5.8389, tangentMode=RCTM_Break, time=0.1009, value=0.7898}
         {arriveTangent=-0.8784, interpMode=RCIM_Cubic, leaveTangent=-0.8784, time=0.4971, value=0.008}
         <at 1 0>
        ]}}})
    Scale RGB=&VectorFromCurve(VectorCurve=DataInterface() {propertyValues {XCurve {keys [<at 0 0.4583>,<at 1 1>]}, YCurve {keys [<at 0 0.4583>,<at 1 1>]}, ZCurve {keys [<at 0 0.4583>,<at 1 1>]}}})
    ScaleA=true
    ScaleRGB=true
   ) {enabled=false}
   &SolveForcesAndVelocity(Clamp Velocity=false, Limit Acceleration=false)
   &SubUVAnimation(
    End Frame Range Override=62
    "Sprite Renderer "=DataInterface() {propertyValues {SpriteRenderer=&self:SmokeLoop_trail_stand001_0.NiagaraSpriteRendererProperties_3}}
    Start Frame Range Override=63
    UseEndFrame=false
    UseStartFrame=false
   )
   &KillParticlesInVolume(
    Box Size:Vector3f=&Subtract_Vector(A=<60 60 60>, B=<-60 -60 -60>)
    Kill Shape=::NewEnumerator1
    Volume Origin:Vector3f=<0 0 0>
   ) {enabled=false}
  ]
  rendererClasses [&NiagaraSpriteRendererProperties]
  renderers [{rendererClass=&NiagaraSpriteRendererProperties}]
 }
}
summary {
 systemName=NS_Rhino_Backpack_Smoke_Converted
 emitters [
  {emitterName=SmokeLoop_trail_stand001, rendererClasses [&NiagaraSpriteRendererProperties]}
  {emitterName=SmokeLoop_stand002, rendererClasses [&NiagaraSpriteRendererProperties]}
  {emitterName=SmokeLoop_move001, rendererClasses [&NiagaraSpriteRendererProperties]}
  {emitterName=SmokeLoop_stand, rendererClasses [&NiagaraSpriteRendererProperties]}
 ]
 userVariables []
}
