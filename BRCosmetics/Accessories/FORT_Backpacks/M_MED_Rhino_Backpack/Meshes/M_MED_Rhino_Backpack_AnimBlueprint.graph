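# Animation Blueprint for the Rhino backpack cosmetic that adds secondary motion to its bags, shovel, pipe and pistons.
# Starting from the local ref pose, chained Anim Dynamics nodes simulate constrained, gravity-overridden motion for bags and shovel, plus a zero-gravity linear spring for the pipe.
# Transform Bone nodes then additively translate three left/right piston pairs in bone space, each pair sharing one PistonMovement vector.
# The EventGraph is empty, so nothing here drives those vectors.
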AnimBlueprint /BRCosmetics/Accessories/FORT_Backpacks/M_MED_Rhino_Backpack/Meshes/M_MED_Rhino_Backpack_AnimBlueprint
target_skeleton=&/M_MED_Rhino_Backpack_Skeleton
graphs [
 AnimGraph {nodes [
   n1 = Anim Dynamics - Bone: dyn_Left_Bag(ComponentPose=Local To Component(LocalPose=Local Space Ref Pose()), Alpha=1, ExternalForce="0,0,0") {
    angularDampingOverride=0.5
    angularSpringConstant=100
    bAngularSpring=true
    bGravityOverrideInSimSpace=true
    bOverrideAngularDamping=true
    bOverrideLinearDamping=true
    bUseGravityOverride=true
    linearDampingOverride=0.8
    alphaBoolBlend {blendOption=Linear}
    boundBone {boneName=dyn_Left_Bag}
    componentAppliedLinearAccClamp=<5000 5000 5000>
    componentLinearAccScale=<0.8 0.8 0.8>
    componentLinearVelScale=<0.8 0.8 0.8>
    gravityOverride {z=-980}
    physicsBodyDefinitions [{
      sphereCollisionRadius=0
      boundBone {boneName=dyn_Left_Bag}
      boxExtents=<8 15 10>
      constraintSetup {angularLimitsMax=<0 10 0>, angularLimitsMin=<0 -5 0>, angularTarget=<0 0 0>, linearAxesMax=<0 0 1.5>, linearAxesMin=<0 0 -1.5>}
      localJointOffset=<-1 0 -6>
     }]
   }
   n2 = Anim Dynamics - Bone: dyn_Right_Bag(ComponentPose=@n1, Alpha=1, ExternalForce="0,0,0") {
    angularDampingOverride=0.5
    angularSpringConstant=100
    bAngularSpring=true
    bGravityOverrideInSimSpace=true
    bOverrideAngularDamping=true
    bOverrideLinearDamping=true
    bUseGravityOverride=true
    linearDampingOverride=0.8
    alphaBoolBlend {blendOption=Linear}
    boundBone {boneName=dyn_Right_Bag}
    componentAppliedLinearAccClamp=<5000 5000 5000>
    componentLinearAccScale=<0.8 0.8 0.8>
    componentLinearVelScale=<0.8 0.8 0.8>
    gravityOverride {z=-980}
    physicsBodyDefinitions [{
      sphereCollisionRadius=0
      boundBone {boneName=dyn_Right_Bag}
      boxExtents=<8 15 10>
      constraintSetup {angularLimitsMax=<0 5 0>, angularLimitsMin=<0 -10 0>, angularTarget=<0 0 0>, linearAxesMax=<0 0 1.5>, linearAxesMin=<0 0 -1.5>}
      localJointOffset=<1 0 -6>
     }]
   }
   n3 = Anim Dynamics - Bone: dyn_Shovel(ComponentPose=@n2, Alpha=1, ExternalForce="0,0,0") {
    bGravityOverrideInSimSpace=true
    bUseGravityOverride=true
    alphaBoolBlend {blendOption=Linear}
    boundBone {boneName=dyn_Shovel}
    componentLinearAccScale=<0.3 0.3 0.3>
    componentLinearVelScale=<0.3 0.3 0.3>
    gravityOverride {z=-980}
    physicsBodyDefinitions [{
      sphereCollisionRadius=0
      boundBone {boneName=dyn_Shovel}
      boxExtents=<20 5 15>
      constraintSetup {angularTargetAxis=AxisY, twistAxis=AxisY, angularLimitsMax=<5 0 0>, angularLimitsMin=<-10 0 0>, angularTarget=<0 0 0>}
      localJointOffset=<0 -0.1 -1.5>
     }]
   }
   n4 = Anim Dynamics - Bone: dyn_Pipe(ComponentPose=@n3, Alpha=1, ExternalForce="0,0,0") {
    bGravityOverrideInSimSpace=true
    bLinearSpring=true
    bUseGravityOverride=true
    gravityScale=0
    linearSpringConstant=500
    alphaBoolBlend {blendOption=Linear}
    boundBone {boneName=dyn_Pipe}
    componentLinearAccScale=<1 1 1>
    componentLinearVelScale=<1 1 1>
    gravityOverride {z=-980}
    physicsBodyDefinitions [{
      sphereCollisionRadius=0
      boundBone {boneName=dyn_Pipe}
      boxExtents=<6 8 5>
      constraintSetup {angularLimitsMax=<0 0 0>, angularLimitsMin=<0 0 0>, angularTarget=<0 0 0>, linearAxesMax=<0.5 0.1 0.5>, linearAxesMin=<-0.5 -0.1 -0.5>}
      localJointOffset=<0 0 0>
     }]
   }
   n5 = Get Piston1Movement() -> (Piston1Movement:Vector)
   n6 = "Transform (Modify) Bone - Bone: dyn_Left_Piston_01"(ComponentPose=@n4, Alpha=1, Translation=@n5, Rotation="0,0,0", Scale="1,1,1") {
    translationMode=BMM_Additive
    translationSpace=BCS_BoneSpace
    alphaBoolBlend {blendOption=Linear}
    boneToModify {boneName=dyn_Left_Piston_01}
   }
   n7 = "Transform (Modify) Bone - Bone: dyn_Right_Piston_01"(ComponentPose=@n6, Alpha=1, Translation=@n5, Rotation="0,0,0", Scale="1,1,1") {
    translationMode=BMM_Additive
    translationSpace=BCS_BoneSpace
    alphaBoolBlend {blendOption=Linear}
    boneToModify {boneName=dyn_Right_Piston_01}
   }
   n8 = Get Piston2Movement() -> (Piston2Movement:Vector)
   n9 = "Transform (Modify) Bone - Bone: dyn_Left_Piston_02"(ComponentPose=@n7, Alpha=1, Translation=@n8, Rotation="0,0,0", Scale="1,1,1") {
    translationMode=BMM_Additive
    translationSpace=BCS_BoneSpace
    alphaBoolBlend {blendOption=Linear}
    boneToModify {boneName=dyn_Left_Piston_02}
   }
   n10 = "Transform (Modify) Bone - Bone: dyn_Right_Piston_02"(ComponentPose=@n9, Alpha=1, Translation=@n8, Rotation="0,0,0", Scale="1,1,1") {
    translationMode=BMM_Additive
    translationSpace=BCS_BoneSpace
    alphaBoolBlend {blendOption=Linear}
    boneToModify {boneName=dyn_Right_Piston_02}
   }
   n11 = Get Piston3Movement() -> (Piston3Movement:Vector)
   n12 = "Transform (Modify) Bone - Bone: dyn_Left_Piston_03"(ComponentPose=@n10, Alpha=1, Translation=@n11, Rotation="0,0,0", Scale="1,1,1") {
    translationMode=BMM_Additive
    translationSpace=BCS_BoneSpace
    alphaBoolBlend {blendOption=Linear}
    boneToModify {boneName=dyn_Left_Piston_03}
   }
   n13 = "Transform (Modify) Bone - Bone: dyn_Right_Piston_03"(ComponentPose=@n12, Alpha=1, Translation=@n11, Rotation="0,0,0", Scale="1,1,1") {
    translationMode=BMM_Additive
    translationSpace=BCS_BoneSpace
    alphaBoolBlend {blendOption=Linear}
    boneToModify {boneName=dyn_Right_Piston_03}
   }
   Output Pose(Result=Component To Local(ComponentPose=@n13)) {layerGroup=None}
   AnimGraphNode_SubInput_0 = Input Pose()
  ]}
 EventGraph {links [], nodes []}
]
variables []
